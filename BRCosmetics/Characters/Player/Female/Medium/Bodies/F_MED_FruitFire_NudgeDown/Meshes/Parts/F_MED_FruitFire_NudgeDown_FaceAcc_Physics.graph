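# Physics asset for the face accessory part of the FruitFire NudgeDown female player cosmetic, covering a hanging hair chain attached to the head.
# A small head capsule anchors three dyn_hairDown bones, each constrained to the one before it.
# The hair bodies are short, very wide, almost spherical capsules.
# The constraints allow broad swing, widest in the second axis at the last joint, while twist is almost locked.
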PhysicsAsset /BRCosmetics/Characters/Player/Female/Medium/Bodies/F_MED_FruitFire_NudgeDown/Meshes/Parts/F_MED_FruitFire_NudgeDown_FaceAcc_Physics
bodies {
 dyn_hairDown_01 [{shapeType=Capsule, length=2.5, radius=14.5, center=<7.3558 0 0>, rotation=<rot -90 0 0>}]
 dyn_hairDown_02 [{shapeType=Capsule, length=2.5, radius=14.5, center=<7.3558 0 0>, rotation=<rot -90 0 0>}]
 dyn_hairDown_03 [{shapeType=Capsule, length=0, radius=14.5, center=<1.9962 0 0>, rotation=<rot -90 0 0>}]
 head [{shapeType=Capsule, length=5, radius=5}]
}
constraints [
 {bone1Name=dyn_hairDown_01, bone2Name=head, swing1LimitDegrees=70, swing2LimitDegrees=45, twistLimitDegrees=1}
 {bone1Name=dyn_hairDown_02, bone2Name=dyn_hairDown_01, swing1LimitDegrees=70, swing2LimitDegrees=45, twistLimitDegrees=1}
 {bone1Name=dyn_hairDown_03, bone2Name=dyn_hairDown_02, swing1LimitDegrees=70, swing2LimitDegrees=70, twistLimitDegrees=1}
]
